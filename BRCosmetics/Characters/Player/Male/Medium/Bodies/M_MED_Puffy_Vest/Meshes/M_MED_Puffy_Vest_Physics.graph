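# Physics asset for the male medium Puffy Vest player body, providing capsule bodies for secondary motion of drawstrings, shoelaces and armband.
# Pelvis, thigh, calf and foot capsules form the lower body, linked by free-rotation constraints.
# Drawstring chains and the armband knot attach freely to the pelvis, and shoelace roots to the calves.
# Within each accessory, child segments have limited swing and tight twist, with shoelaces branching and carrying an extra strand on the right.
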PhysicsAsset /BRCosmetics/Characters/Player/Male/Medium/Bodies/M_MED_Puffy_Vest/Meshes/M_MED_Puffy_Vest_Physics
bodies {
 calf_l [{shapeType=Capsule, length=24.6727, radius=5.3745, center=<-26.9751 0.235 -0.1674>, rotation=<rot 0 86.1064 90.7415>}]
 calf_r [{shapeType=Capsule, length=24.6727, radius=5.3745, center=<26.9751 -0.235 0.1674>, rotation=<rot 0 86.1064 -89.2585>}]
 dyn_armband_hi [{shapeType=Capsule, length=3.9862, radius=1.4424, center=<0.3847 3.8163 -0.2363>, rotation=<rot 0 -175.9689 -80.8298>}]
 dyn_armband_knot [{shapeType=Capsule, length=1.9657, radius=1.6436, center=<-0.0033 0.5713 0.201>, rotation=<rot 0 -51.6166 -72.0407>}]
 dyn_armband_low [{shapeType=Capsule, length=5.8269, radius=1.459, center=<-0.9002 3.9465 2.9304>, rotation=<rot 0 -166.1568 -58.3997>}]
 dyn_drawstring_01_l [{shapeType=Capsule, length=5, radius=0.505, center=<3 0 0>, rotation=<rot 90 0 0>}]
 dyn_drawstring_01_r [{shapeType=Capsule, length=5, radius=0.505, center=<3 0 0>, rotation=<rot 90 0 0>}]
 dyn_drawstring_02_l [{shapeType=Capsule, length=4.5944, radius=0.9106, center=<3 0 0>, rotation=<rot 90 0 0>}]
 dyn_drawstring_02_r [{shapeType=Capsule, length=4.5944, radius=0.9106, center=<3 0 0>, rotation=<rot 90 0 0>}]
 dyn_drawstring_03_l [{shapeType=Capsule, length=4.4398, radius=1.4028, center=<3 0 0>, rotation=<rot 90 0 0>}]
 dyn_drawstring_03_r [{shapeType=Capsule, length=5, radius=0.8406, center=<3 0 0>, rotation=<rot 90 0 0>}]
 dyn_shoestring_A_01_l [{shapeType=Capsule, length=4, radius=1.6119, center=<3 0 0>, rotation=<rot 90 0 0>}]
 dyn_shoestring_A_01_r [{shapeType=Capsule, length=4, radius=1.0152, center=<3 0 0>, rotation=<rot 90 0 0>}]
 dyn_shoestring_A_02_l [{shapeType=Capsule, length=4, radius=0.505, center=<3 0 0>, rotation=<rot 90 0 0>}]
 dyn_shoestring_A_02_r [{shapeType=Capsule, length=4, radius=1.2729, center=<3 0 0>, rotation=<rot 90 0 0>}]
 dyn_shoestring_B_01_l [{shapeType=Capsule, length=4, radius=1.0774, center=<3 0 0>, rotation=<rot 90 0 0>}]
 dyn_shoestring_B_01_r [{shapeType=Capsule, length=4, radius=0.8698, center=<3 0 0>, rotation=<rot 90 0 0>}]
 dyn_shoestring_B_02_l [{shapeType=Capsule, length=4, radius=0.505, center=<3 0 0>, rotation=<rot 90 0 0>}]
 dyn_shoestring_B_02_r [{shapeType=Capsule, length=4, radius=1.027, center=<3 0 0>, rotation=<rot 90 0 0>}]
 dyn_shoestring_C_01_r [{shapeType=Capsule, length=4, radius=0.5434, center=<3 0 0>, rotation=<rot 90 0 0>}]
 dyn_shoestring_C_02_r [{shapeType=Capsule, length=4, radius=0.3055, center=<3 0 0>, rotation=<rot 90 0 0>}]
 dyn_shoestring_l [{shapeType=Capsule, length=4, radius=0.6624, center=<3 0 0>, rotation=<rot 90 0 0>}]
 dyn_shoestring_r [{shapeType=Capsule, length=4, radius=0.6747, center=<3 0 0>, rotation=<rot 90 0 0>}]
 foot_l [{shapeType=Capsule, length=14.3057, radius=4.6929, center=<-2.8481 5.998 -0.3658>, rotation=<rot 0 24.9986 -90>}]
 foot_r [{shapeType=Capsule, length=14.3057, radius=4.6929, center=<2.8481 -5.998 0.3658>, rotation=<rot 0 24.9987 89.9999>}]
 pelvis [
  {shapeType=Capsule, length=11.1729, radius=10.8644, center=<-7.0685 0.1958 0>}
  {shapeType=Capsule, length=13.0835, radius=8.3482, center=<-2.0105 1.9993 0>}
 ]
 thigh_l [{shapeType=Capsule, length=12.4643, radius=9.4472, center=<-8.9796 1.7022 1.0034>, rotation=<rot 90 0 -179.9999>}]
 thigh_r [{shapeType=Capsule, length=7.6197, radius=9.8528, center=<8.8746 -1.7022 -0.0005>, rotation=<rot 90 0 0>}]
}
constraints [
 {bone1Name=dyn_shoestring_C_02_r, bone2Name=dyn_shoestring_C_01_r, swing1LimitDegrees=20, swing2LimitDegrees=20, twistLimitDegrees=1}
 {bone1Name=dyn_shoestring_A_02_r, bone2Name=dyn_shoestring_A_01_r, swing1LimitDegrees=20, swing2LimitDegrees=20, twistLimitDegrees=1}
 {bone1Name=dyn_shoestring_B_02_r, bone2Name=dyn_shoestring_B_01_r, swing1LimitDegrees=20, swing2LimitDegrees=20, twistLimitDegrees=1}
 {bone1Name=dyn_shoestring_C_01_r, bone2Name=dyn_shoestring_r, swing1LimitDegrees=20, swing2LimitDegrees=20, twistLimitDegrees=1}
 {bone1Name=dyn_shoestring_A_01_r, bone2Name=dyn_shoestring_r, swing1LimitDegrees=20, swing2LimitDegrees=20, twistLimitDegrees=1}
 {bone1Name=dyn_shoestring_B_01_r, bone2Name=dyn_shoestring_r, swing1LimitDegrees=20, swing2LimitDegrees=20, twistLimitDegrees=1}
 {bone1Name=dyn_shoestring_A_01_l, bone2Name=dyn_shoestring_l, swing1LimitDegrees=20, swing2LimitDegrees=20, twistLimitDegrees=1}
 {bone1Name=dyn_shoestring_A_02_l, bone2Name=dyn_shoestring_A_01_l, swing1LimitDegrees=20, swing2LimitDegrees=20, twistLimitDegrees=1}
 {bone1Name=dyn_shoestring_B_01_l, bone2Name=dyn_shoestring_l, swing1LimitDegrees=20, swing2LimitDegrees=20, twistLimitDegrees=1}
 {bone1Name=dyn_shoestring_B_02_l, bone2Name=dyn_shoestring_B_01_l, swing1LimitDegrees=20, swing2LimitDegrees=20, twistLimitDegrees=1}
 {bone1Name=dyn_drawstring_02_r, bone2Name=dyn_drawstring_01_r, swing1LimitDegrees=45, swing2LimitDegrees=45, twistLimitDegrees=1}
 {bone1Name=dyn_drawstring_03_r, bone2Name=dyn_drawstring_02_r, swing1LimitDegrees=45, swing2LimitDegrees=45, twistLimitDegrees=1}
 {bone1Name=dyn_drawstring_02_l, bone2Name=dyn_drawstring_01_l, swing1LimitDegrees=45, swing2LimitDegrees=45, twistLimitDegrees=1}
 {bone1Name=dyn_drawstring_03_l, bone2Name=dyn_drawstring_02_l, swing1LimitDegrees=45, swing2LimitDegrees=45, twistLimitDegrees=1}
 {
  bone1Name=foot_r
  bone2Name=calf_r
  swing1LimitDegrees=45
  swing1Motion=Free
  swing2LimitDegrees=45
  swing2Motion=Free
  twistLimitDegrees=45
  twistMotion=Free
 }
 {
  bone1Name=dyn_shoestring_r
  bone2Name=calf_r
  swing1LimitDegrees=45
  swing1Motion=Free
  swing2LimitDegrees=45
  swing2Motion=Free
  twistLimitDegrees=45
  twistMotion=Free
 }
 {
  bone1Name=foot_l
  bone2Name=calf_l
  swing1LimitDegrees=45
  swing1Motion=Free
  swing2LimitDegrees=45
  swing2Motion=Free
  twistLimitDegrees=45
  twistMotion=Free
 }
 {
  bone1Name=dyn_shoestring_l
  bone2Name=calf_l
  swing1LimitDegrees=45
  swing1Motion=Free
  swing2LimitDegrees=45
  swing2Motion=Free
  twistLimitDegrees=45
  twistMotion=Free
 }
 {
  bone1Name=calf_l
  bone2Name=thigh_l
  swing1LimitDegrees=45
  swing1Motion=Free
  swing2LimitDegrees=45
  swing2Motion=Free
  twistLimitDegrees=45
  twistMotion=Free
 }
 {
  bone1Name=calf_r
  bone2Name=thigh_r
  swing1LimitDegrees=45
  swing1Motion=Free
  swing2LimitDegrees=45
  swing2Motion=Free
  twistLimitDegrees=45
  twistMotion=Free
 }
 {
  bone1Name=dyn_drawstring_01_r
  bone2Name=pelvis
  swing1LimitDegrees=45
  swing1Motion=Free
  swing2LimitDegrees=45
  swing2Motion=Free
  twistLimitDegrees=45
  twistMotion=Free
 }
 {
  bone1Name=dyn_drawstring_01_l
  bone2Name=pelvis
  swing1LimitDegrees=45
  swing1Motion=Free
  swing2LimitDegrees=45
  swing2Motion=Free
  twistLimitDegrees=45
  twistMotion=Free
 }
 {
  bone1Name=thigh_r
  bone2Name=pelvis
  swing1LimitDegrees=45
  swing1Motion=Free
  swing2LimitDegrees=45
  swing2Motion=Free
  twistLimitDegrees=45
  twistMotion=Free
 }
 {
  bone1Name=thigh_l
  bone2Name=pelvis
  swing1LimitDegrees=45
  swing1Motion=Free
  swing2LimitDegrees=45
  swing2Motion=Free
  twistLimitDegrees=45
  twistMotion=Free
 }
 {
  bone1Name=dyn_armband_knot
  bone2Name=pelvis
  swing1LimitDegrees=45
  swing1Motion=Free
  swing2LimitDegrees=45
  swing2Motion=Free
  twistLimitDegrees=45
  twistMotion=Free
 }
 {bone1Name=dyn_armband_low, bone2Name=dyn_armband_knot, swing1LimitDegrees=10, swing2LimitDegrees=10, twistLimitDegrees=2}
 {bone1Name=dyn_armband_hi, bone2Name=dyn_armband_knot, swing1LimitDegrees=10, swing2LimitDegrees=10, twistLimitDegrees=2}
]
