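# Post-process-style Anim Blueprint for the Puffy Vest body cosmetic, layering secondary motion onto the pose copied from the attached parent mesh.
# Flap nodes and additive bone corrections shape the collar from neck rotation plus head and right-clavicle alphas, computed against spine_05 each update.
# AnimDynamics, weighted multi-bone constraints and a final RigidBody pass handle drawstrings, shoestrings and physics.
# Several computed look and tilt alphas never reach the AnimGraph.
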AnimBlueprint /BRCosmetics/Characters/Player/Male/Medium/Bodies/M_MED_Puffy_Vest/Meshes/M_MED_Puffy_Vest_AnimBP
target_skeleton=&/M_MED_Puffy_Vest_Skeleton
graphs [
 AnimGraph {nodes [
   n1 = Flap Node - Bone: dyn_collar_back!(
    ComponentPose:Component Space Pose Link Structure=Local To Component(LocalPose=Copy Pose From Mesh() {bUseAttachedParent=true})
    Alpha:float=1
   ) -> (Pose:Component Space Pose Link Structure) {
    is=title FortAnimGraphNode_Flap
    bUseClamp=true
    targetClampMin=-3.14
    targetComponent=RotationZ
    alphaBoolBlend {blendOption=Linear}
    sourceDrives [{
      bUseRange=false
      drivingCurve=null
      multiplier=-1
      rangeMax=1
      rangeMin=-1
      remappedMax=1
      remappedMin=0
      sourceComponent=RotationZ
      useQuaternion=true
      sourceBone {boneName=neck_01}
     }]
    targetBone {boneName=dyn_collar_back}
   }
   n2 = Flap Node - Bone: dyn_collar_back(ComponentPose=@n1, Alpha=1) {
    bUseClamp=true
    targetClampMin=-10
    targetComponent=TranslationY
    alphaBoolBlend {blendOption=Linear}
    sourceDrives [{
      bUseRange=false
      drivingCurve=null
      multiplier=-10
      rangeMax=1
      rangeMin=-1
      remappedMax=1
      remappedMin=0
      sourceComponent=RotationZ
      useQuaternion=true
      sourceBone {boneName=neck_01}
     }]
    targetBone {boneName=dyn_collar_back}
   }
   n3 = Get look_up_alpha() -> (look_up_alpha:double)
   n4 = "Transform (Modify) Bone - Bone: dyn_collar_02_l"(ComponentPose=@n2, Alpha=float * float(A=@n3, B=0.5), Translation="0,0,0", Rotation="12.193,-42.115,-5.118", Scale="1,1,1") {
    rotationMode=BMM_Additive
    rotationSpace=BCS_BoneSpace
    alphaBoolBlend {blendOption=Linear}
    boneToModify {boneName=dyn_collar_02_l}
    rotation=<rot 12.193 -42.115 -5.118>
   }
   n5 = "Transform (Modify) Bone - Bone: dyn_collar_back"(ComponentPose=@n4, Alpha=@n3, Translation="-0.957,-1.291,0.000", Rotation="0.000,-40,-10", Scale="1,1,1") {
    rotationMode=BMM_Additive
    rotationSpace=BCS_BoneSpace
    translationMode=BMM_Additive
    translationSpace=BCS_BoneSpace
    alphaBoolBlend {blendOption=Linear}
    boneToModify {boneName=dyn_collar_back}
    rotation {roll=-10, yaw=-40}
    translation=<-0.9569 -1.2909 0>
   }
   n6 = "Transform (Modify) Bone - Bone: dyn_collar_01_r"(ComponentPose=@n5, Alpha=0, Translation="0,0,0", Rotation="0,0,0", Scale="1,1,1") {rotationMode=BMM_Additive, rotationSpace=BCS_BoneSpace, alphaBoolBlend {blendOption=Linear}, boneToModify {boneName=dyn_collar_01_r}}
   n7 = "Transform (Modify) Bone - Bone: dyn_collar_02_r"(ComponentPose=@n6, Alpha=@n3, Translation="0,0,0", Rotation="-4.251,-26.190,20.967", Scale="1,1,1") {
    rotationMode=BMM_Additive
    rotationSpace=BCS_BoneSpace
    alphaBoolBlend {blendOption=Linear}
    boneToModify {boneName=dyn_collar_02_r}
    rotation=<rot -4.251 -26.19 20.967>
   }
   n8 = "Transform (Modify) Bone - Bone: dyn_collar_02_l"(ComponentPose=@n7, Alpha=Get look_left_alpha() -> (look_left_alpha:double), Translation="0,1,1", Rotation="0,-30,0", Scale="1,1,1") {
    rotationMode=BMM_Additive
    rotationSpace=BCS_BoneSpace
    translationMode=BMM_Additive
    translationSpace=BCS_BoneSpace
    alphaBoolBlend {blendOption=Linear}
    boneToModify {boneName=dyn_collar_02_l}
    rotation {yaw=-30}
    translation {y=1, z=1}
   }
   n9 = "Transform (Modify) Bone - Bone: dyn_collar_01_l"(ComponentPose=@n8, Alpha=1, Translation="0,1,0", Rotation="0,-30,0", Scale="1,1,1") {
    rotationMode=BMM_Additive
    rotationSpace=BCS_BoneSpace
    translationMode=BMM_Additive
    translationSpace=BCS_BoneSpace
    alphaBoolBlend {blendOption=Linear}
    boneToModify {boneName=dyn_collar_01_l}
    rotation {yaw=-30}
    translation {y=1}
   }
   n10 = "Transform (Modify) Bone - Bone: dyn_collar_02_r"(ComponentPose=@n9, Alpha=1, Translation="0,0,0", Rotation="0,0,0", Scale="1,1,1") {
    rotationMode=BMM_Additive
    rotationSpace=BCS_BoneSpace
    translationMode=BMM_Additive
    translationSpace=BCS_BoneSpace
    alphaBoolBlend {blendOption=Linear}
    boneToModify {boneName=dyn_collar_02_r}
   }
   n11 = Get clavicle_up_r_alpha() -> (clavicle_up_r_alpha:double)
   n12 = "Transform (Modify) Bone - Bone: dyn_collar_02_r"(ComponentPose=@n10, Alpha=@n11, Translation="-0.245,-1.823,-0.675", Rotation="-0.847,-33.679,-3.619", Scale="1,1,1") {
    rotationMode=BMM_Additive
    rotationSpace=BCS_BoneSpace
    translationMode=BMM_Additive
    translationSpace=BCS_BoneSpace
    alphaBoolBlend {blendOption=Linear}
    boneToModify {boneName=dyn_collar_02_r}
    rotation=<rot -0.847 -33.6791 -3.619>
    translation=<-0.245 -1.823 -0.675>
   }
   n13 = "Transform (Modify) Bone - Bone: dyn_collar_back"(ComponentPose=@n12, Alpha=@n11, Translation="0,0,0", Rotation="0,0,0", Scale="1,1,1") {
    rotationMode=BMM_Additive
    rotationSpace=BCS_BoneSpace
    translationMode=BMM_Additive
    translationSpace=BCS_BoneSpace
    alphaBoolBlend {blendOption=Linear}
    boneToModify {boneName=dyn_collar_back}
   }
   n14 = "Transform (Modify) Bone - Bone: dyn_collar_01_r"(ComponentPose=@n13, Alpha=@n11, Translation="0,0,0", Rotation="-1.400,-25.973,4.027", Scale="1,1,1") {
    rotationMode=BMM_Additive
    rotationSpace=BCS_BoneSpace
    translationMode=BMM_Additive
    translationSpace=BCS_BoneSpace
    alphaBoolBlend {blendOption=Linear}
    boneToModify {boneName=dyn_collar_01_r}
    rotation=<rot -1.4 -25.9728 4.027>
   }
   n15 = Anim Dynamics - Bone: dyn_drawstring_r(ComponentPose=@n14, Alpha=1, ExternalForce="0,0,0") {
    angularDampingOverride=1
    bGravityOverrideInSimSpace=true
    bOverrideAngularDamping=true
    gravityScale=0.1
    simulationSpace=RootRelative
    alphaBoolBlend {blendOption=Linear}
    boundBone {boneName=dyn_drawstring_r}
    physicsBodyDefinitions [{
      sphereCollisionRadius=0
      boundBone {boneName=dyn_drawstring_r}
      boxExtents=<5 10 10>
      constraintSetup {angularLimitsMax=<0 20 5>, angularLimitsMin=<0 0 -60>, angularTarget=<0 0 0>}
      localJointOffset=<5 0.2 0>
     }]
   }
   n16 = Anim Dynamics - Bone: dyn_drawstring_l(ComponentPose=@n15, Alpha=1, ExternalForce="0,0,0") {
    angularDampingOverride=1
    bGravityOverrideInSimSpace=true
    bOverrideAngularDamping=true
    gravityScale=0.1
    simulationSpace=RootRelative
    alphaBoolBlend {blendOption=Linear}
    boundBone {boneName=dyn_drawstring_l}
    physicsBodyDefinitions [{
      sphereCollisionRadius=0
      boundBone {boneName=dyn_drawstring_l}
      boxExtents=<5 10 10>
      constraintSetup {angularLimitsMax=<0 0 5>, angularLimitsMin=<0 -20 -60>, angularTarget=<0 0 0>}
      localJointOffset=<5 0.2 0>
     }]
   }
   n17 = "Constraint - dyn_shoestring_l (PPP)"!(ComponentPose=@n16, Alpha=1, ConstraintWeights_0=0.3941, ConstraintWeights_1:float=0.5185, ConstraintWeights_2:float=0.0874) -> (Pose) {
    alphaBoolBlend {blendOption=Linear}
    boneToModify {boneName=dyn_shoestring_l}
    constraintSetup [{targetBone {boneName=calf_l}}, {targetBone {boneName=foot_l}}, {targetBone {boneName=calf_twist_02_l}}]
    constraintWeights [1,1,1]
   }
   n18 = "Constraint - dyn_shoestring_r (PPPP)"!(
    ComponentPose=@n17
    Alpha=1
    ConstraintWeights_0=0
    ConstraintWeights_1:float=0.0885
    ConstraintWeights_2:float=0.5141
    ConstraintWeights_3:float=0.3974
   ) -> (Pose) {
    alphaBoolBlend {blendOption=Linear}
    boneToModify {boneName=dyn_shoestring_r}
    constraintSetup [{targetBone {boneName=ball_r}}, {targetBone {boneName=calf_twist_02_r}}, {targetBone {boneName=foot_r}}, {targetBone {boneName=calf_r}}]
    constraintWeights [1,1,1,1]
   }
   n19 = "Constraint - dyn_drawstring_01_l (PPPPP)"!(
    ComponentPose=@n18
    Alpha=1
    ConstraintWeights_0=0.0781
    ConstraintWeights_1:float=0.0266
    ConstraintWeights_2:float=0.0178
    ConstraintWeights_3:float=0.0216
    ConstraintWeights_4:float=0.8559
   ) -> (Pose) {
    alphaBoolBlend {blendOption=Linear}
    boneToModify {boneName=dyn_drawstring_01_l}
    constraintSetup [
     {targetBone {boneName=spine_01}}
     {targetBone {boneName=spine_02}}
     {targetBone {boneName=spine_03}}
     {targetBone {boneName=thigh_twist_01_l}}
     {targetBone {boneName=pelvis}}
    ]
    constraintWeights [1,1,1,1,1]
   }
   n20 = "Constraint - dyn_drawstring_01_r (PPPPP)"!(
    ComponentPose=@n19
    Alpha=1
    ConstraintWeights_0=0.0781
    ConstraintWeights_1:float=0.0266
    ConstraintWeights_2:float=0.0178
    ConstraintWeights_3:float=0.0216
    ConstraintWeights_4:float=0.8559
   ) -> (Pose) {
    alphaBoolBlend {blendOption=Linear}
    boneToModify {boneName=dyn_drawstring_01_r}
    constraintSetup [
     {targetBone {boneName=spine_01}}
     {targetBone {boneName=spine_02}}
     {targetBone {boneName=spine_03}}
     {targetBone {boneName=thigh_twist_01_r}}
     {targetBone {boneName=pelvis}}
    ]
    constraintWeights [1,1,1,1,1]
   }
   n21 = RigidBody(ComponentPose=@n20) {
    bClampLinearTranslationLimitToRefPose=true
    bOverrideWorldGravity=true
    overridePhysicsAsset=&/M_MED_Puffy_Vest_Physics
    simulationSpace=BaseBoneSpace
    alphaBoolBlend {blendOption=Linear}
    baseBoneRef {boneName=root}
    overrideWorldGravity {z=-980}
   }
   Output Pose(Result=Component To Local(ComponentPose=@n21)) {layerGroup=None}
  ]}
 EventGraph {nodes [
   Event BlueprintUpdateAnimation()
   > n1 = Sequence!(execute) -> (then_0, then_1, then_2:Exec, then_3:Exec)
   n2 = Get PartSkelMesh() -> (PartSkelMesh:Skeletal Mesh Component Object Reference)
   n3 = BreakRotator(InRot=BreakTransform(InTransform=GetDeltaTransformFromRefPose(self=@n2, BoneName=head, BaseName=spine_05)).Rotation)
   Set look_left_alpha(execute=@n1.then_2, look_left_alpha:double=MapRangeClamped(Value=@n3, InRangeA=0, InRangeB=40)) -> (Output_Get:double)
   > Set look_right_alpha(look_right_alpha:double=MapRangeClamped(Value=@n3, InRangeA=0, InRangeB=-20)) -> (Output_Get:double)
   n4 = Get PartSkelMesh()
   n5 = BreakRotator(InRot=BreakTransform(InTransform=GetDeltaTransformFromRefPose(self=@n4, BoneName=head, BaseName=spine_05)).Rotation)
   Set look_down_alpha(execute=@n1.then_1, look_down_alpha:double=MapRangeClamped(Value=@n5.Yaw, InRangeA=-10, InRangeB=25)) -> (Output_Get:double)
   > Set look_up_alpha(look_up_alpha:double=MapRangeClamped(Value=@n5.Yaw, InRangeA=10, InRangeB=-40, OutRangeA=0.1)) -> (Output_Get:double)
   n6 = Get PartSkelMesh()
   n7 = BreakRotator(InRot=BreakTransform(InTransform=GetDeltaTransformFromRefPose(self=@n6, BoneName=head, BaseName=spine_05)).Rotation)
   Set tilt_left_alpha(execute=@n1.then_3, tilt_left_alpha:double=MapRangeClamped(Value=@n7.Pitch, InRangeA=10, InRangeB=30)) -> (Output_Get:double)
   > Set tilt_right_alpha(tilt_right_alpha:double=MapRangeClamped(Value=@n7.Pitch, InRangeA=0, InRangeB=-10)) -> (Output_Get:double)
   n8 = "Max (Float)"(A=Get look_left_alpha(), B=0)
   > Set left_max_alpha(left_max_alpha:double=@n8) -> (Output_Get:double)
   > Set left_max_down(left_max_down:double="Max (Float)"(A=@n8, B=Get look_down_alpha() -> (look_down_alpha:double))) -> (Output_Get:double)
   n9 = "Max (Float)"(A=Get look_right_alpha() -> (look_right_alpha:double), B=0)
   > Set right_max_alpha(right_max_alpha:double=@n9) -> (Output_Get:double)
   > Set right_max_down(right_max_down:double="Max (Float)"(A=@n9, B=Get look_down_alpha())) -> (Output_Get:double)
   n10 = BreakTransform(InTransform=GetDeltaTransformFromRefPose(self=Get PartSkelMesh(), BoneName=clavicle_r, BaseName=spine_05))
   Set clavicle_up_r_alpha(execute=@n1.then_0, clavicle_up_r_alpha:double=MapRangeClamped(Value=BreakRotator(InRot=@n10.Rotation).Pitch, InRangeA=0, InRangeB=19)) -> (Output_Get:double)
   Event BlueprintInitializeAnimation()
   > Set PartSkelMesh(PartSkelMesh:Skeletal Mesh Component Object Reference=GetOwnerSkeletalMeshComponent()) -> (Output_Get:Skeletal Mesh Component Object Reference)
   n11 = GetPartSkeletalMeshComponent()
   n12 = BreakRotator(InRot=BreakTransform(InTransform=GetDeltaTransformFromRefPose(self=@n11, BoneName=head, BaseName=spine_05)).Rotation)
   Set head_back(head_back:double=MapRangeClamped(Value=@n12.Yaw, InRangeA=0, InRangeB=-40)) -> (Output_Get:double)
   > Set head_left(head_left:double=0) -> (Output_Get:double)
   > Set head_right(head_right:double=MapRangeClamped(Value=@n12.Pitch, InRangeA=0, InRangeB=40)) -> (Output_Get:double)
   GetPartSkeletalMeshComponent()
   Get tilt_left_alpha() -> (tilt_left_alpha:double)
   Get tilt_right_alpha() -> (tilt_right_alpha:double)
  ]}
]
variables [
 clavicle_up_r_alpha {type=number}
 head_back {type=number}
 head_left {type=number}
 head_right {type=number}
 left_max_alpha {type=number}
 left_max_down {type=number}
 look_down_alpha {type=number}
 look_left_alpha {type=number}
 look_right_alpha {type=number}
 look_up_alpha {type=number}
 PartSkelMesh {type=SkeletalMeshComponent}
 right_max_alpha {type=number}
 right_max_down {type=number}
 tilt_left_alpha {type=number}
 tilt_right_alpha {type=number}
]
