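# Post-copy AnimBP for the male puffy vest body's face accessory mesh, adding secondary motion to hair bangs and side strips.
# It copies the attached parent's pose, then chains spring-based Anim Dynamics through three bangs bones and two strips.
# While targeting (bIsTargeting), the update event lowers bangs physics weight and blends in an additive rotation on the first bangs bone.
# The strips always simulate at full weight.
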AnimBlueprint /BRCosmetics/Characters/Player/Male/Medium/Bodies/M_MED_Puffy_Vest/Meshes/Parts/M_MED_Puffy_Vest_FaceAcc_AnimBP
target_skeleton=&/M_MED_Puffy_Vest_FaceAcc_Skeleton
graphs [
 AnimGraph {nodes [
   n1 = Get ads_alpha() -> (ads_alpha:double)
   n2 = Anim Dynamics - Bone: dyn_hair_bangs_01(ComponentPose=Local To Component(LocalPose=Copy Pose From Mesh() {bUseAttachedParent=true}), Alpha=@n1, ExternalForce="0,0,0") {
    angularDampingOverride=1
    angularSpringConstant=200
    bAngularSpring=true
    bGravityOverrideInSimSpace=true
    bOverrideAngularDamping=true
    gravityScale=0.35
    simulationSpace=RootRelative
    alphaBoolBlend {blendOption=Linear}
    alphaScaleBiasClamp {bInterpResult=true, interpSpeedDecreasing=4, interpSpeedIncreasing=4}
    boundBone {boneName=dyn_hair_bangs_01}
    physicsBodyDefinitions [{
      sphereCollisionRadius=0
      boundBone {boneName=dyn_hair_bangs_01}
      boxExtents=<4 8 8>
      constraintSetup {angularLimitsMax=<0 10 40>, angularLimitsMin=<0 -10 0>, angularTarget=<1 0 0>}
      localJointOffset=<4 0 0>
     }]
   }
   n3 = Anim Dynamics - Bone: dyn_hair_bangs_02(ComponentPose=@n2, Alpha=@n1, ExternalForce="0,0,0") {
    angularDampingOverride=1
    angularSpringConstant=200
    bAngularSpring=true
    bGravityOverrideInSimSpace=true
    bOverrideAngularDamping=true
    gravityScale=0.35
    simulationSpace=RootRelative
    alphaBoolBlend {blendOption=Linear}
    alphaScaleBiasClamp {bInterpResult=true, interpSpeedDecreasing=4, interpSpeedIncreasing=4}
    boundBone {boneName=dyn_hair_bangs_02}
    physicsBodyDefinitions [{
      sphereCollisionRadius=0
      boundBone {boneName=dyn_hair_bangs_02}
      boxExtents=<4 8 8>
      constraintSetup {angularLimitsMax=<0 10 40>, angularLimitsMin=<0 -10 0>, angularTarget=<1 0 0>}
      localJointOffset=<4 0 0>
     }]
   }
   n4 = Anim Dynamics - Bone: dyn_hair_bangs_03(ComponentPose=@n3, Alpha=@n1, ExternalForce="0,0,0") {
    angularDampingOverride=1
    angularSpringConstant=200
    bAngularSpring=true
    bGravityOverrideInSimSpace=true
    bOverrideAngularDamping=true
    gravityScale=0.35
    simulationSpace=RootRelative
    alphaBoolBlend {blendOption=Linear}
    alphaScaleBiasClamp {bInterpResult=true, interpSpeedDecreasing=4, interpSpeedIncreasing=4}
    boundBone {boneName=dyn_hair_bangs_03}
    physicsBodyDefinitions [{
      sphereCollisionRadius=0
      boundBone {boneName=dyn_hair_bangs_03}
      boxExtents=<4 8 8>
      constraintSetup {angularLimitsMax=<0 10 40>, angularLimitsMin=<0 -10 0>, angularTarget=<1 0 0>}
      localJointOffset=<4 0 0>
     }]
   }
   n5 = Anim Dynamics - Bone: dyn_hair_strip_r(ComponentPose=@n4, Alpha=1, ExternalForce="0,0,0") {
    angularDampingOverride=1
    angularSpringConstant=50
    bAngularSpring=true
    bGravityOverrideInSimSpace=true
    bOverrideAngularDamping=true
    gravityScale=0.35
    simulationSpace=RootRelative
    alphaBoolBlend {blendOption=Linear}
    boundBone {boneName=dyn_hair_strip_r}
    physicsBodyDefinitions [{
      sphereCollisionRadius=0
      boundBone {boneName=dyn_hair_strip_r}
      boxExtents=<10 20 20>
      constraintSetup {angularLimitsMax=<0 10 20>, angularLimitsMin=<0 -10 0>, angularTarget=<1 0 0>}
      localJointOffset=<10 0 0>
     }]
   }
   n6 = Anim Dynamics - Bone: dyn_hair_strip_l(ComponentPose=@n5, Alpha=1, ExternalForce="0,0,0") {
    angularDampingOverride=1
    angularSpringConstant=50
    bAngularSpring=true
    bGravityOverrideInSimSpace=true
    bOverrideAngularDamping=true
    gravityScale=0.35
    simulationSpace=RootRelative
    alphaBoolBlend {blendOption=Linear}
    boundBone {boneName=dyn_hair_strip_l}
    physicsBodyDefinitions [{
      sphereCollisionRadius=0
      boundBone {boneName=dyn_hair_strip_l}
      boxExtents=<10 20 20>
      constraintSetup {angularLimitsMax=<0 10 20>, angularLimitsMin=<0 -10 0>, angularTarget=<1 0 0>}
      localJointOffset=<10 0 0>
     }]
   }
   n7 = "Transform (Modify) Bone - Bone: dyn_hair_bangs_01"(
    ComponentPose=@n6
    Alpha=Get ads_alpha_invert() -> (ads_alpha_invert:double)
    Translation="0,0,0"
    Rotation="-1.915,-18.86,-5.62"
    Scale="1,1,1"
   ) {
    rotationMode=BMM_Additive
    rotationSpace=BCS_BoneSpace
    translationMode=BMM_Additive
    translationSpace=BCS_BoneSpace
    alphaBoolBlend {blendOption=Linear}
    boneToModify {boneName=dyn_hair_bangs_01}
    rotation=<rot -1.915 -18.86 -5.62>
   }
   Output Pose(Result=Component To Local(ComponentPose=@n7)) {layerGroup=None}
  ]}
 EventGraph {nodes [
   Event BlueprintUpdateAnimation()
   n1 = Get bIsTargeting()
   > Set ads_alpha(ads_alpha:double=SelectFloat(A=0.1, B=0.5, bPickA=@n1)) -> (Output_Get:double)
   > Set ads_alpha_invert(ads_alpha_invert:double=float * float(A="To Float (Boolean)"(InBool=@n1), B=0.5)) -> (Output_Get:double)
   TryGetPawnOwner()
  ]}
]
variables [
 ads_alpha {type=number}
 ads_alpha_invert {type=number}
]
